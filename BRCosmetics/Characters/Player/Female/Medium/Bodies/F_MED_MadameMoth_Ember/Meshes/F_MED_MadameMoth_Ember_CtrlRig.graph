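# Control Rig for the Madame Moth Ember female body cosmetic that adjusts scale-collar and shoulder-fur bones as the arms move forward.
# Mirrored spherical pose readers on each upper arm write left and right arm-forward alphas.
# Those alphas weight transform modifiers that nudge and shrink each collar bone and shorten the first shoulder-fur bone along X.
# Other dyn bones in the hierarchy, such as arm fluff and leg curtains, are untouched.
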ControlRigBlueprint /BRCosmetics/Characters/Player/Female/Medium/Bodies/F_MED_MadameMoth_Ember/Meshes/F_MED_MadameMoth_Ember_CtrlRig
graphs [
 RigVMFunctionLibrary {links [], nodes []}
 RigVMModel {nodes [
   RigUnit_BeginExecution = &RigUnit_BeginExecution()
   n1 = &RigUnit_SphericalPoseReader(
    ActiveRegionScaleFactors=(PositiveWidth=1,NegativeWidth=3,PositiveHeight=0.1,NegativeHeight=5)
    ActiveRegionSize=0.1
    Debug=(bDrawDebug=false,bDraw2D=False,bDrawLocalAxes=False,DebugScale=25,DebugSegments=20,DebugThickness=0.25)
    DriverAxis=(X=1,Y=0,Z=0)
    DriverItem=(Type=Bone,Name="upperarm_l")
    FalloffRegionScaleFactors=(PositiveWidth=1,NegativeWidth=2,PositiveHeight=1.1,NegativeHeight=1)
    FalloffSize=0.3
    FlipHeightScaling=False
    FlipWidthScaling=False
    OptionalParentItem=(Type=Bone,Name="None")
    RotationOffset=(X=85,Y=30,Z=-10)
    ExecutePin=@RigUnit_BeginExecution.ExecutePin
   )
   > RigVMVariableNode(Value=@n1.OutputParam) {variable=L armForward Alpha}
   > n2 = &RigUnit_SphericalPoseReader(
    ActiveRegionScaleFactors=(PositiveWidth=1,NegativeWidth=3,PositiveHeight=0.1,NegativeHeight=5)
    ActiveRegionSize=0.1
    Debug=(bDrawDebug=false,bDraw2D=False,bDrawLocalAxes=False,DebugScale=25,DebugSegments=20,DebugThickness=0.25)
    DriverAxis=(X=-1,Y=0,Z=0)
    DriverItem=(Type=Bone,Name="upperarm_r")
    FalloffRegionScaleFactors=(PositiveWidth=1,NegativeWidth=2,PositiveHeight=1.1,NegativeHeight=1)
    FalloffSize=0.3
    FlipHeightScaling=False
    FlipWidthScaling=true
    OptionalParentItem=(Type=Bone,Name="None")
    RotationOffset=(X=85,Y=150,Z=170)
   )
   > RigVMVariableNode(Value=@n2.OutputParam) {variable=R armForward Alpha}
   > &RigUnit_ModifyTransforms(
    ItemToModify=((Item=(Type=Bone,Name="dyn_L_scaleCollar"),Transform=(Rotation=(X=0,Y=0,Z=0,W=1),Translation=(X=1,Y=0,Z=0),Scale3D=(X=0.5,Y=0.7,Z=0.9))))
    WeightMinimum=0
    Weight=RigVMVariableNode() {variable=L armForward Alpha}.Value
   )
   > &RigUnit_ModifyTransforms(
    ItemToModify=((Item=(Type=Bone,Name="dyn_R_scaleCollar"),Transform=(Rotation=(X=0,Y=0,Z=0,W=1),Translation=(X=-1,Y=0,Z=0),Scale3D=(X=0.5,Y=0.7,Z=0.9))))
    WeightMinimum=0
    Weight=RigVMVariableNode() {variable=R armForward Alpha}.Value
   )
   > &RigUnit_ModifyTransforms(
    ItemToModify=((Item=(Type=Bone,Name="dyn_L_shoulderFur_1"),Transform=(Rotation=(X=0,Y=0,Z=0,W=1),Translation=(X=0,Y=0,Z=0),Scale3D=(X=0.6,Y=1,Z=1))))
    WeightMinimum=0
    Weight=RigVMVariableNode() {variable=L armForward Alpha}.Value
   )
   > &RigUnit_ModifyTransforms(
    ItemToModify=((Item=(Type=Bone,Name="dyn_R_shoulderFur_1"),Transform=(Rotation=(X=0,Y=0,Z=0,W=1),Translation=(X=0,Y=0,Z=0),Scale3D=(X=0.6,Y=1,Z=1))))
    WeightMinimum=0
    Weight=RigVMVariableNode() {variable=R armForward Alpha}.Value
   )
  ]}
]
hierarchy [
 root {loc=<0 0 0>, rot=<0 0 0 1>}
 pelvis {parent=root, loc=<0 -0.5301 98.0571>, rot=<0.0127 0.707 -0.0127 -0.707>}
 spine_01 {parent=pelvis, loc=<6.1208 -0.9328 0>, rot=<0 0 0.034 -0.9994>}
 spine_02 {parent=spine_01, loc=<6.5996 -0.2095 0>, rot=<0 0 0.0238 -0.9997>}
 spine_03 {parent=spine_02, loc=<6.4914 -0.1237 0>, rot=<0 0 0.0149 -0.9999>}
 spine_04 {parent=spine_03, loc=<6.9598 0.0794 0>, rot=<0 0 -0.0132 -0.9999>}
 spine_05 {parent=spine_04, loc=<6.9987 0 0>, rot=<0 0 -0.0015 -1>}
 attach_cape {parent=spine_05, loc=<9.3216 -8.4501 0>, rot=<0.0616 -0.7044 -0.0616 -0.7044>}
 neck_01 {parent=spine_05, loc=<11.3138 0.1529 0>, rot=<0 0 -0.2003 -0.9797>}
 neck_02 {parent=neck_01, loc=<4.5627 0 0>, rot=<0 0 0.0504 -0.9987>}
 head {parent=neck_02, loc=<5.3914 0 0>, rot=<0 0 0.115 -0.9934>}
 attach_backpack {parent=spine_05, loc=<0 -8.35 0>, rot=<-0.0093 -0.707 0.0093 -0.707>}
 clavicle_l {parent=spine_05, loc=<5.4831 -0.1053 -3.2946>, rot=<0.0465 -0.7289 -0.009 -0.6829>}
 upperarm_l {parent=clavicle_l, loc=<9.8658 -1.2136 -0.4105>, rot=<0.0164 -0.3245 -0.0269 -0.9454>}
 lowerarm_l {parent=upperarm_l, loc=<26.465 -0.8345 0.259>, rot=<0.0036 0.0161 -0.2298 -0.9731>}
 hand_l {parent=lowerarm_l, loc=<22.7355 -0.3203 -0.6105>, rot=<-0.6784 -0.0223 0.0735 -0.7307>}
 weapon_l {parent=hand_l, loc=<-0.1974 -3.4111 0.3811>, rot=<0 0 0 1>}
 thumb_01_l {parent=hand_l, loc=<1.3118 -0.3191 -2.459>, rot=<0.6546 -0.4042 0.0664 -0.6354>}
 thumb_02_l {parent=thumb_01_l, loc=<5.9336 -0.0448 0.0401>, rot=<-0.0037 0.0213 0.298 -0.9543>}
 thumb_03_l {parent=thumb_02_l, loc=<2.2223 0.3682 -0.0907>, rot=<0.0118 -0.013 0.0066 -0.9998>}
 index_metacarpal_l {parent=hand_l, loc=<1.6535 0.0183 -2.96>, rot=<0.0158 -0.0124 -0.0173 -0.9996>}
 index_01_l {parent=index_metacarpal_l, loc=<8.3008 0.0357 -0.0524>, rot=<-0.0044 0.0485 0.0924 -0.9945>}
 index_02_l {parent=index_01_l, loc=<3.3587 -0.3004 0.0452>, rot=<-0.002 0.0027 0.3225 -0.9466>}
 index_03_l {parent=index_02_l, loc=<2.5578 -0.123 -0.0252>, rot=<-0.002 0.0067 0.2558 -0.9667>}
 middle_metacarpal_l {parent=hand_l, loc=<2.1856 -0.1725 -0.5984>, rot=<0.006 0.0168 -0.0329 -0.9993>}
 middle_01_l {parent=middle_metacarpal_l, loc=<7.3275 -0.2447 0.0129>, rot=<-0.0208 0.0042 0.0858 -0.9961>}
 middle_02_l {parent=middle_01_l, loc=<4.0123 0.1401 0.1175>, rot=<-0.0218 0.0258 0.3512 -0.9357>}
 middle_03_l {parent=middle_02_l, loc=<2.6077 0.0534 0.1373>, rot=<0.0082 0.0188 0.1708 -0.9851>}
 ring_metacarpal_l {parent=hand_l, loc=<2.2074 -0.3625 1.3638>, rot=<-0.0284 0.0722 -0.0458 -0.9959>}
 ring_01_l {parent=ring_metacarpal_l, loc=<7.0295 -0.6197 -0.1446>, rot=<-0.0489 -0.0252 0.1892 -0.9804>}
 ring_02_l {parent=ring_01_l, loc=<3.6174 0.4647 -0.1027>, rot=<0 0 0.2806 -0.9598>}
 ring_03_l {parent=ring_02_l, loc=<2.4152 0.114 0.0199>, rot=<0 0 0.1189 -0.9929>}
 pinky_metacarpal_l {parent=hand_l, loc=<1.7616 -0.9753 3.3814>, rot=<-0.0397 0.0814 -0.069 -0.9935>}
 pinky_01_l {parent=pinky_metacarpal_l, loc=<7.1709 -0.4831 -0.1094>, rot=<-0.0654 -0.0035 0.2242 -0.9723>}
 pinky_02_l {parent=pinky_01_l, loc=<2.8274 0.0893 -0.1043>, rot=<0.0037 0.0011 0.2854 -0.9584>}
 pinky_03_l {parent=pinky_02_l, loc=<2.2046 0.0307 -0.0078>, rot=<-0.0002 0.0098 0.1445 -0.9895>}
 hand_attach_l {parent=hand_l, loc=<8.5 -3 0>, rot=<0.7071 0.7071 0 0>}
 lowerarm_twist_02_l {parent=lowerarm_l, loc=<10.6707 0 0>, rot=<0 0 0 1>}
 lowerarm_twist_01_l {parent=lowerarm_l, loc=<18.316 0 0>, rot=<0 0 0 1>}
 dyn_L_armFluff_1 {parent=lowerarm_l, loc=<2.4188 6.1129 7.0004>, rot=<0.982 -0.0247 -0.125 0.1395>}
 dyn_L_armFluff_2 {parent=lowerarm_l, loc=<3.3019 4.186 8.92>, rot=<0.9953 0.0859 -0.029 0.0353>}
 dyn_L_armFluff_3 {parent=lowerarm_l, loc=<7.6838 2.419 9.0789>, rot=<0.9675 -0.2004 -0.1456 -0.0504>}
 dyn_L_armFluff_4 {parent=lowerarm_l, loc=<0.9931 1.8864 8.2261>, rot=<0.9904 -0.102 0.0313 -0.0878>}
 dyn_L_armFluff_5 {parent=lowerarm_l, loc=<2.5353 0.1755 8.8694>, rot=<0.9891 0.0302 -0.0236 -0.1423>}
 dyn_L_armFluff_6 {parent=lowerarm_l, loc=<7.7856 -1.5013 8.2534>, rot=<0.9645 -0.0776 -0.1496 -0.2035>}
 dyn_L_armFluff_7 {parent=lowerarm_l, loc=<0.9673 -1.0744 6.855>, rot=<0.9868 0.012 -0.0101 -0.1609>}
 dyn_L_armFluff_8 {parent=lowerarm_l, loc=<3.4624 -4.5826 5.1462>, rot=<0.9154 0.0235 -0.0122 -0.4017>}
 upperarm_twist_01_l {parent=upperarm_l, loc=<7.2135 -0.1134 0>, rot=<0 0 0 1>}
 upperarm_twist_02_l {parent=upperarm_l, loc=<19.7998 -0.1134 0>, rot=<0 0 0 1>}
 dyn_L_shoulderFur_1 {parent=clavicle_l, loc=<6.3645 -0.1028 7.2263>, rot=<0.0162 0.0328 0.073 -0.9967>}
 dyn_L_shoulderFur_2 {parent=dyn_L_shoulderFur_1, loc=<7.6947 -1.1048 0.4958>, rot=<0.0604 0.1887 0.0006 -0.9802>}
 dyn_L_scaleCollar {parent=clavicle_l, loc=<7.2815 7.8424 2.8599>, rot=<-0.1134 -0.2617 0.9302 0.2312>}
 clavicle_r {parent=spine_05, loc=<5.4831 -0.1053 3.2946>, rot=<0.7289 0.0465 -0.6829 0.009>}
 upperarm_r {parent=clavicle_r, loc=<-9.8658 1.2136 0.4105>, rot=<0.0164 -0.3245 -0.0269 -0.9454>}
 lowerarm_r {parent=upperarm_r, loc=<-26.465 0.8345 -0.259>, rot=<0.0036 0.0161 -0.2298 -0.9731>}
 hand_r {parent=lowerarm_r, loc=<-22.7355 0.3203 0.6106>, rot=<-0.6784 -0.0223 0.0735 -0.7307>}
 weapon_r {parent=hand_r, loc=<0.1974 3.4111 -0.3811>, rot=<0 0 0 1>}
 thumb_01_r {parent=hand_r, loc=<-1.3118 0.3191 2.459>, rot=<0.6546 -0.4042 0.0664 -0.6354>}
 thumb_02_r {parent=thumb_01_r, loc=<-5.9336 0.0448 -0.0401>, rot=<-0.0037 0.0213 0.298 -0.9543>}
 thumb_03_r {parent=thumb_02_r, loc=<-2.2223 -0.3682 0.0907>, rot=<0.0118 -0.013 0.0066 -0.9998>}
 index_metacarpal_r {parent=hand_r, loc=<-1.6535 -0.0183 2.96>, rot=<0.0158 -0.0124 -0.0173 -0.9996>}
 index_01_r {parent=index_metacarpal_r, loc=<-8.3008 -0.0357 0.0524>, rot=<-0.0044 0.0485 0.0924 -0.9945>}
 index_02_r {parent=index_01_r, loc=<-3.3587 0.3004 -0.0452>, rot=<-0.002 0.0027 0.3225 -0.9466>}
 index_03_r {parent=index_02_r, loc=<-2.5578 0.123 0.0252>, rot=<-0.002 0.0067 0.2558 -0.9667>}
 middle_metacarpal_r {parent=hand_r, loc=<-2.1856 0.1725 0.5984>, rot=<0.006 0.0168 -0.0329 -0.9993>}
 middle_01_r {parent=middle_metacarpal_r, loc=<-7.3275 0.2447 -0.0129>, rot=<-0.0208 0.0042 0.0858 -0.9961>}
 middle_02_r {parent=middle_01_r, loc=<-4.0123 -0.1401 -0.1175>, rot=<-0.0218 0.0258 0.3512 -0.9357>}
 middle_03_r {parent=middle_02_r, loc=<-2.6077 -0.0534 -0.1373>, rot=<0.0082 0.0188 0.1708 -0.9851>}
 ring_metacarpal_r {parent=hand_r, loc=<-2.2074 0.3625 -1.3638>, rot=<-0.0284 0.0722 -0.0458 -0.9959>}
 ring_01_r {parent=ring_metacarpal_r, loc=<-7.0295 0.6197 0.1446>, rot=<-0.0489 -0.0252 0.1892 -0.9804>}
 ring_02_r {parent=ring_01_r, loc=<-3.6174 -0.4647 0.1027>, rot=<0 0 0.2806 -0.9598>}
 ring_03_r {parent=ring_02_r, loc=<-2.4152 -0.114 -0.0199>, rot=<0 0 0.1189 -0.9929>}
 pinky_metacarpal_r {parent=hand_r, loc=<-1.7616 0.9753 -3.3814>, rot=<-0.0397 0.0814 -0.069 -0.9935>}
 pinky_01_r {parent=pinky_metacarpal_r, loc=<-7.1709 0.4831 0.1094>, rot=<-0.0654 -0.0035 0.2242 -0.9723>}
 pinky_02_r {parent=pinky_01_r, loc=<-2.8274 -0.0893 0.1043>, rot=<0.0037 0.0011 0.2854 -0.9584>}
 pinky_03_r {parent=pinky_02_r, loc=<-2.2046 -0.0307 0.0078>, rot=<-0.0002 0.0098 0.1445 -0.9895>}
 hand_attach_r {parent=hand_r, loc=<-8.5001 2.9999 0>, rot=<0 0 -0.7071 -0.7071>}
 lowerarm_twist_02_r {parent=lowerarm_r, loc=<-10.6707 0 0>, rot=<0 0 0 -1>}
 lowerarm_twist_01_r {parent=lowerarm_r, loc=<-18.316 0 0>, rot=<0 0 0 -1>}
 dyn_R_armFluff_1 {parent=lowerarm_r, loc=<-2.4189 -6.1129 -7.0003>, rot=<0.982 -0.0247 -0.125 0.1395>}
 dyn_R_armFluff_2 {parent=lowerarm_r, loc=<-3.3016 -4.1862 -8.9202>, rot=<0.9953 0.0859 -0.029 0.0353>}
 dyn_R_armFluff_3 {parent=lowerarm_r, loc=<-7.6839 -2.4189 -9.0788>, rot=<0.9675 -0.2004 -0.1456 -0.0504>}
 dyn_R_armFluff_4 {parent=lowerarm_r, loc=<-0.9931 -1.8864 -8.2261>, rot=<0.9904 -0.102 0.0313 -0.0878>}
 dyn_R_armFluff_5 {parent=lowerarm_r, loc=<-2.5353 -0.1755 -8.8695>, rot=<0.9891 0.0302 -0.0236 -0.1423>}
 dyn_R_armFluff_6 {parent=lowerarm_r, loc=<-7.7855 1.5013 -8.2535>, rot=<0.9645 -0.0776 -0.1496 -0.2035>}
 dyn_R_armFluff_7 {parent=lowerarm_r, loc=<-0.9673 1.0744 -6.855>, rot=<0.9868 0.012 -0.0101 -0.1609>}
 dyn_R_armFluff_8 {parent=lowerarm_r, loc=<-3.4622 4.5825 -5.1465>, rot=<0.9154 0.0235 -0.0122 -0.4017>}
 upperarm_twist_01_r {parent=upperarm_r, loc=<-7.2135 0.1134 0>, rot=<0 0 0 -1>}
 upperarm_twist_02_r {parent=upperarm_r, loc=<-19.7998 0.1134 0>, rot=<0 0 0 -1>}
 dyn_R_shoulderFur_1 {parent=clavicle_r, loc=<-6.3645 0.1028 -7.2263>, rot=<0.0162 0.0328 0.073 -0.9967>}
 dyn_R_shoulderFur_2 {parent=dyn_R_shoulderFur_1, loc=<-7.6947 1.1048 -0.4958>, rot=<0.0604 0.1887 0.0006 -0.9802>}
 dyn_R_scaleCollar {parent=clavicle_r, loc=<-7.2815 -7.8424 -2.8601>, rot=<-0.1134 -0.2617 0.9302 0.2312>}
 dyn_back_collarFluff {parent=spine_05, loc=<11.4507 -11.216 0>, rot=<0.1181 -0.6972 0.6972 -0.1181>}
 dyn_collarConstraint {parent=spine_05, loc=<11.3138 -2.8585 0>, rot=<0 0 -0.2003 -0.9797>}
 dyn_R_backJacket {parent=spine_04, loc=<2.1584 -8.413 7.1869>, rot=<0.0414 -0.9953 0.0871 -0.0036>}
 dyn_L_backJacket {parent=spine_04, loc=<2.1582 -8.4129 -7.1869>, rot=<0.0414 -0.9953 -0.0871 0.0036>}
 thigh_l {parent=pelvis, loc=<-1.0855 -0.2109 -8.7053>, rot=<0.0009 0.0355 0.0155 -0.9993>}
 calf_l {parent=thigh_l, loc=<-42.5284 -0.0019 -0.7297>, rot=<0.0006 -0.0161 -0.0391 -0.9991>}
 foot_l {parent=calf_l, loc=<-46.7401 -2.5014 -0.592>, rot=<0.0427 -0.0238 0.0353 -0.9982>}
 ball_l {parent=foot_l, loc=<-3.5658 10.8066 -0.1544>, rot=<0.0001 -0.0001 -0.7063 -0.7079>}
 calf_twist_02_l {parent=calf_l, loc=<-19.1546 -1.7943 -0.2638>, rot=<0.0237 0.0095 -0.0035 -0.9997>}
 calf_twist_01_l {parent=calf_l, loc=<-35.8483 -1.9197 -0.5794>, rot=<0.0237 0.0095 -0.0035 -0.9997>}
 dyn_L_legCurtain_outer_1 {parent=calf_l, loc=<-18.0811 -0.6142 -4.6721>, rot=<0.0088 0.706 0.7064 0.0498>}
 dyn_L_legCurtain_outer_2 {parent=dyn_L_legCurtain_outer_1, loc=<17.1993 0 0>, rot=<0 0 0 1>}
 dyn_L_legCurtain_inner_1 {parent=calf_l, loc=<-17.5629 -0.6852 4.4282>, rot=<-0.0156 -0.708 0.7047 0.043>}
 dyn_L_legCurtain_inner_2 {parent=dyn_L_legCurtain_inner_1, loc=<17.1993 0 0>, rot=<0 0 0 1>}
 dyn_L_legCurtain_front_1 {parent=calf_l, loc=<-17.4398 4.3297 -0.1431>, rot=<0.0194 -0.0023 0.999 0.0414>}
 dyn_L_legCurtain_front_2 {parent=dyn_L_legCurtain_front_1, loc=<17.1993 0 -0.8468>, rot=<0 0 0 1>}
 dyn_L_legCurtain_front_footFold {parent=dyn_L_legCurtain_front_2, loc=<12.287 1.3293 -1.2463>, rot=<0.0254 -0.036 0.5757 -0.8164>}
 dyn_L_legCurtain_back_1 {parent=calf_l, loc=<-18.3913 -5.4523 -0.1677>, rot=<0.0623 -0.998 -0.0055 0.0007>}
 dyn_L_legCurtain_back_2 {parent=dyn_L_legCurtain_back_1, loc=<17.1993 0.9855 0>, rot=<0 0 0 1>}
 thigh_twist_01_l {parent=thigh_l, loc=<-27.952 0 0>, rot=<0.0013 0 0 -1>}
 thigh_r {parent=pelvis, loc=<-1.0855 -0.2109 8.7053>, rot=<0.0355 -0.0009 0.9993 0.0155>}
 calf_r {parent=thigh_r, loc=<42.5284 0.0019 0.7297>, rot=<0.0006 -0.0161 -0.0391 -0.9991>}
 foot_r {parent=calf_r, loc=<46.7401 2.5014 0.592>, rot=<0.0427 -0.0238 0.0353 -0.9982>}
 ball_r {parent=foot_r, loc=<3.5658 -10.8066 0.1544>, rot=<0.0001 -0.0001 -0.7063 -0.7079>}
 calf_twist_02_r {parent=calf_r, loc=<19.1546 1.7943 0.2638>, rot=<0.0237 0.0095 -0.0035 -0.9997>}
 calf_twist_01_r {parent=calf_r, loc=<35.8483 1.9197 0.5794>, rot=<0.0237 0.0095 -0.0035 -0.9997>}
 dyn_R_legCurtain_outer_1 {parent=calf_r, loc=<18.081 0.6142 4.6721>, rot=<0.0088 0.706 0.7064 0.0498>}
 dyn_R_legCurtain_outer_2 {parent=dyn_R_legCurtain_outer_1, loc=<-17.1993 0 0>, rot=<0 0 0 -1>}
 dyn_R_legCurtain_inner_1 {parent=calf_r, loc=<17.5629 0.6851 -4.4282>, rot=<-0.0156 -0.708 0.7047 0.043>}
 dyn_R_legCurtain_inner_2 {parent=dyn_R_legCurtain_inner_1, loc=<-17.1993 0 0>, rot=<0 0 0 1>}
 dyn_R_legCurtain_front_1 {parent=calf_r, loc=<17.4398 -4.3297 0.1431>, rot=<0.0194 -0.0023 0.999 0.0414>}
 dyn_R_legCurtain_front_2 {parent=dyn_R_legCurtain_front_1, loc=<-17.1993 0 0.8468>, rot=<0 0 0 1>}
 dyn_R_legCurtain_front_footFold {parent=dyn_R_legCurtain_front_2, loc=<-12.287 -1.3293 1.2462>, rot=<0.0254 -0.036 0.5757 -0.8164>}
 dyn_R_legCurtain_back_1 {parent=calf_r, loc=<18.3913 5.4523 0.1677>, rot=<0.0623 -0.998 -0.0055 0.0007>}
 dyn_R_legCurtain_back_2 {parent=dyn_R_legCurtain_back_1, loc=<-17.1993 -0.9855 0>, rot=<0 0 0 1>}
 thigh_twist_01_r {parent=thigh_r, loc=<27.952 0 0>, rot=<0.0013 0 0 -1>}
 dyn_holster {parent=pelvis, loc=<-1.8929 1.0475 -15.1851>, rot=<-0.0178 -0.9947 0.0018 -0.1014>}
 attach {parent=root, loc=<0 0 0>, rot=<0 0 0 1>}
 ik_hand_root {parent=root, loc=<0 0 0>, rot=<0 0 0 1>}
 ik_hand_gun {parent=ik_hand_root, loc=<-48.1743 3.3545 104.4381>, rot=<0.6276 0.1309 0.4183 -0.6434>}
 ik_hand_r {parent=ik_hand_gun, loc=<0 0 0>, rot=<0 0 0 1>}
 ik_hand_l {parent=ik_hand_gun, loc=<59.3271 67.6997 34.3524>, rot=<-0.6158 -0.7027 -0.3565 0>}
 ik_foot_root {parent=root, loc=<0 0 0>, rot=<0 0 0 1>}
 ik_foot_l {parent=ik_foot_root, loc=<14.8262 -7.3545 8.296>, rot=<0.0357 0.7045 0.0284 -0.7082>}
 ik_foot_r {parent=ik_foot_root, loc=<-14.8262 -7.3545 8.296>, rot=<0.7082 0.0284 -0.7045 0.0357>}
 attach_fp {parent=root, loc=<0 0 0>, rot=<0 0 0 1>}
 camera_root {parent=root, loc=<0 0 0>, rot=<0 0 0 1>}
 camera {parent=camera_root, loc=<0 0 0>, rot=<0 0 0 1>}
 attach_emote_01 {parent=root, loc=<0 0 0>, rot=<0 0 0 1>}
 attach_emote_02 {parent=root, loc=<0 0 0>, rot=<0 0 0 1>}
 attach_emote_03 {parent=root, loc=<0 0 0>, rot=<0 0 0 1>}
 warp_point_01 {parent=root, loc=<0 0 0>, rot=<0 0 0 1>}
 warp_point_02 {parent=root, loc=<0 0 0>, rot=<0 0 0 1>}
 warp_point_03 {parent=root, loc=<0 0 0>, rot=<0 0 0 1>}
]
variables [
 L armForward Alpha {type=float, default=0}
 R armForward Alpha {type=float, default=0}
]
