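# Animation Blueprint for the Madame Moth Ember cosmetic body, adding secondary motion to fur, jacket, leg curtains and attachments.
# It copies the parent mesh pose, applies bone constraints and Anim Dynamics, then runs a Control Rig.
# A front-end idle curve blends DefaultSlot onto the holster.
# The event graph updates simulation space and constraint values, and feeds foot velocity, calf collision and skydiving wind to material-based WPO cloth.
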AnimBlueprint /BRCosmetics/Characters/Player/Female/Medium/Bodies/F_MED_MadameMoth_Ember/Meshes/F_MED_MadameMoth_Ember_AnimBP
target_skeleton=&/F_MED_MadameMoth_Ember_Skeleton
graphs [
 AnimGraph {nodes [
   "Constraint - attach_cape (PPPPP)"!(
    ComponentPose
    Alpha=1
    ConstraintWeights_0=0.0032
    ConstraintWeights_1:float=0.3915
    ConstraintWeights_2:float=0.0257
    ConstraintWeights_3:float=0.0357
    ConstraintWeights_4:float=0.5439
   ) -> (Pose) {
    alphaBoolBlend {blendOption=Linear}
    boneToModify {boneName=attach_cape}
    constraintSetup [
     {targetBone {boneName=spine_04}}
     {targetBone {boneName=spine_05}}
     {targetBone {boneName=clavicle_l}}
     {targetBone {boneName=clavicle_r}}
     {targetBone {boneName=dyn_back_collarFluff}}
    ]
    constraintWeights [1,1,1,1,1]
   }
   "Constraint - attach_backpack (PPPPPPPP)"!(
    ComponentPose
    Alpha=1
    ConstraintWeights_0=0.0194
    ConstraintWeights_1:float=0.1204
    ConstraintWeights_2:float=0.3992
    ConstraintWeights_3:float=0.0411
    ConstraintWeights_4:float=0.0448
    ConstraintWeights_5:float=0.1642
    ConstraintWeights_6:float=0.1633
    ConstraintWeights_7:float=0.0475
   ) -> (Pose) {
    alphaBoolBlend {blendOption=Linear}
    boneToModify {boneName=attach_backpack}
    constraintSetup [
     {targetBone {boneName=spine_03}}
     {targetBone {boneName=spine_04}}
     {targetBone {boneName=spine_05}}
     {targetBone {boneName=clavicle_l}}
     {targetBone {boneName=clavicle_r}}
     {targetBone {boneName=dyn_R_backJacket}}
     {targetBone {boneName=dyn_L_backJacket}}
     {targetBone {boneName=dyn_back_collarFluff}}
    ]
    constraintWeights [1,1,1,1,1,1,1,1]
   }
   Output Pose(Result=Layered blend per bone(
     BasePose=Use cached pose 'startpose'()
     BlendPoses_0=Slot 'DefaultSlot'()
     BlendWeights_0=Get Custom Idle is Playing() -> (Custom Idle is Playing:float)
    ) {layerSetup [{branchFilters [{boneName=dyn_holster}]}]}) {layerGroup=None}
   n1 = "Constraint - attach_cape (PPPP)"!(
    ComponentPose=Local To Component(LocalPose=Copy Pose From Mesh() {bUseAttachedParent=true})
    Alpha=1
    ConstraintWeights_0=0.0032
    ConstraintWeights_1:float=0.9354
    ConstraintWeights_2:float=0.0257
    ConstraintWeights_3:float=0.0357
   ) -> (Pose) {
    alphaBoolBlend {blendOption=Linear}
    boneToModify {boneName=attach_cape}
    constraintSetup [{targetBone {boneName=spine_04}}, {targetBone {boneName=spine_05}}, {targetBone {boneName=clavicle_l}}, {targetBone {boneName=clavicle_r}}]
    constraintWeights [1,1,1,1]
   }
   n2 = "Constraint - attach_backpack (PPPPP)"!(
    ComponentPose=@n1
    Alpha=1
    ConstraintWeights_0=0.0194
    ConstraintWeights_1:float=0.4479
    ConstraintWeights_2:float=0.4468
    ConstraintWeights_3:float=0.0411
    ConstraintWeights_4:float=0.0448
   ) -> (Pose) {
    alphaBoolBlend {blendOption=Linear}
    boneToModify {boneName=attach_backpack}
    constraintSetup [
     {targetBone {boneName=spine_03}}
     {targetBone {boneName=spine_04}}
     {targetBone {boneName=spine_05}}
     {targetBone {boneName=clavicle_l}}
     {targetBone {boneName=clavicle_r}}
    ]
    constraintWeights [1,1,1,1,1]
   }
   n3 = Anim Dynamics - Bone: dyn_L_armFluff_1!(SimulationSpace:AnimPhysSimSpaceType Enum, ComponentPose=@n2, Alpha=1, ExternalForce="0,0,0") -> (Pose) {
    angularDampingOverride=1
    angularSpringConstant=10000
    bAngularSpring=true
    bOverrideAngularDamping=true
    bUseGravityOverride=true
    gravityScale=0.35
    simulationSpace=RootRelative
    alphaBoolBlend {blendOption=Linear}
    boundBone {boneName=dyn_L_armFluff_1}
    componentLinearAccScale=<1 1 1>
    componentLinearVelScale=<1 1 1>
    gravityOverride {z=-980}
    physicsBodyDefinitions [{
      boundBone {boneName=dyn_L_armFluff_1}
      boxExtents=<25 25 25>
      constraintSetup {angularLimitsMax=<0 0 30>, angularLimitsMin=<0 -45 -30>, angularTarget=<1 0 0>}
      localJointOffset=<-2 0 0>
     }]
   }
   n4 = Anim Dynamics - Bone: dyn_L_armFluff_2!(SimulationSpace:AnimPhysSimSpaceType Enum, ComponentPose=@n3, Alpha=1, ExternalForce="0,0,0") -> (Pose) {
    angularDampingOverride=1
    angularSpringConstant=12000
    bAngularSpring=true
    bOverrideAngularDamping=true
    bUseGravityOverride=true
    gravityScale=0.35
    simulationSpace=RootRelative
    alphaBoolBlend {blendOption=Linear}
    boundBone {boneName=dyn_L_armFluff_2}
    componentLinearAccScale=<1 1 1>
    componentLinearVelScale=<1 1 1>
    gravityOverride {z=-1200}
    physicsBodyDefinitions [{
      boundBone {boneName=dyn_L_armFluff_2}
      boxExtents=<20 20 20>
      constraintSetup {angularLimitsMax=<0 0 30>, angularLimitsMin=<0 -45 -30>, angularTarget=<1 0 0>}
      localJointOffset=<-2 0 0>
     }]
   }
   n5 = Anim Dynamics - Bone: dyn_L_armFluff_3!(SimulationSpace:AnimPhysSimSpaceType Enum, ComponentPose=@n4, Alpha=1, ExternalForce="0,0,0") -> (Pose) {
    angularDampingOverride=1
    angularSpringConstant=10000
    bAngularSpring=true
    bOverrideAngularDamping=true
    bUseGravityOverride=true
    gravityScale=0.35
    simulationSpace=RootRelative
    alphaBoolBlend {blendOption=Linear}
    boundBone {boneName=dyn_L_armFluff_3}
    componentLinearAccScale=<1 1 1>
    componentLinearVelScale=<1 1 1>
    gravityOverride {z=-980}
    physicsBodyDefinitions [{
      boundBone {boneName=dyn_L_armFluff_3}
      boxExtents=<25 25 25>
      constraintSetup {angularLimitsMax=<0 0 30>, angularLimitsMin=<0 -45 -30>, angularTarget=<1 0 0>}
      localJointOffset=<-2 0 0>
     }]
   }
   n6 = Anim Dynamics - Bone: dyn_L_armFluff_4!(SimulationSpace:AnimPhysSimSpaceType Enum, ComponentPose=@n5, Alpha=1, ExternalForce="0,0,0") -> (Pose) {
    angularDampingOverride=1
    angularSpringConstant=12000
    bAngularSpring=true
    bOverrideAngularDamping=true
    bUseGravityOverride=true
    gravityScale=0.35
    simulationSpace=RootRelative
    alphaBoolBlend {blendOption=Linear}
    boundBone {boneName=dyn_L_armFluff_4}
    componentLinearAccScale=<1 1 1>
    componentLinearVelScale=<1 1 1>
    gravityOverride {z=-980}
    physicsBodyDefinitions [{
      boundBone {boneName=dyn_L_armFluff_4}
      boxExtents=<20 20 20>
      constraintSetup {angularLimitsMax=<0 0 30>, angularLimitsMin=<0 -45 -30>, angularTarget=<1 0 0>}
      localJointOffset=<-2 0 0>
     }]
   }
   n7 = Anim Dynamics - Bone: dyn_L_armFluff_5!(SimulationSpace:AnimPhysSimSpaceType Enum, ComponentPose=@n6, Alpha=1, ExternalForce="0,0,0") -> (Pose) {
    angularDampingOverride=1
    angularSpringConstant=10000
    bAngularSpring=true
    bOverrideAngularDamping=true
    bUseGravityOverride=true
    gravityScale=0.35
    simulationSpace=RootRelative
    alphaBoolBlend {blendOption=Linear}
    boundBone {boneName=dyn_L_armFluff_5}
    componentLinearAccScale=<1 1 1>
    componentLinearVelScale=<1 1 1>
    gravityOverride {z=-980}
    physicsBodyDefinitions [{
      boundBone {boneName=dyn_L_armFluff_5}
      boxExtents=<25 25 25>
      constraintSetup {angularLimitsMax=<0 0 30>, angularLimitsMin=<0 -45 -30>, angularTarget=<1 0 0>}
      localJointOffset=<-2 0 0>
     }]
   }
   n8 = Anim Dynamics - Bone: dyn_L_armFluff_6!(SimulationSpace:AnimPhysSimSpaceType Enum, ComponentPose=@n7, Alpha=1, ExternalForce="0,0,0") -> (Pose) {
    angularDampingOverride=1
    angularSpringConstant=12000
    bAngularSpring=true
    bOverrideAngularDamping=true
    bUseGravityOverride=true
    gravityScale=0.35
    simulationSpace=RootRelative
    alphaBoolBlend {blendOption=Linear}
    boundBone {boneName=dyn_L_armFluff_6}
    componentLinearAccScale=<1 1 1>
    componentLinearVelScale=<1 1 1>
    gravityOverride {z=-980}
    physicsBodyDefinitions [{
      boundBone {boneName=dyn_L_armFluff_6}
      boxExtents=<18 18 18>
      constraintSetup {angularLimitsMax=<0 0 30>, angularLimitsMin=<0 -45 -30>, angularTarget=<1 0 0>}
      localJointOffset=<-2 0 0>
     }]
   }
   n9 = Anim Dynamics - Bone: dyn_L_armFluff_7!(SimulationSpace:AnimPhysSimSpaceType Enum, ComponentPose=@n8, Alpha=1, ExternalForce="0,0,0") -> (Pose) {
    angularDampingOverride=1
    angularSpringConstant=10000
    bAngularSpring=true
    bOverrideAngularDamping=true
    bUseGravityOverride=true
    gravityScale=0.35
    simulationSpace=RootRelative
    alphaBoolBlend {blendOption=Linear}
    boundBone {boneName=dyn_L_armFluff_7}
    componentLinearAccScale=<1 1 1>
    componentLinearVelScale=<1 1 1>
    gravityOverride {z=-1200}
    physicsBodyDefinitions [{
      boundBone {boneName=dyn_L_armFluff_7}
      boxExtents=<25 25 25>
      constraintSetup {angularLimitsMax=<0 0 30>, angularLimitsMin=<0 -45 -30>, angularTarget=<1 0 0>}
      localJointOffset=<-2 0 0>
     }]
   }
   n10 = Anim Dynamics - Bone: dyn_L_armFluff_8!(SimulationSpace:AnimPhysSimSpaceType Enum, ComponentPose=@n9, Alpha=1, ExternalForce="0,0,0") -> (Pose) {
    angularDampingOverride=1
    angularSpringConstant=12000
    bAngularSpring=true
    bOverrideAngularDamping=true
    bUseGravityOverride=true
    gravityScale=0.35
    simulationSpace=RootRelative
    alphaBoolBlend {blendOption=Linear}
    boundBone {boneName=dyn_L_armFluff_8}
    componentLinearAccScale=<1 1 1>
    componentLinearVelScale=<1 1 1>
    gravityOverride {z=-980}
    physicsBodyDefinitions [{
      boundBone {boneName=dyn_L_armFluff_8}
      boxExtents=<18 18 18>
      constraintSetup {angularLimitsMax=<0 0 30>, angularLimitsMin=<0 -45 -30>, angularTarget=<1 0 0>}
      localJointOffset=<-2 0 0>
     }]
   }
   n11 = Anim Dynamics - Bone: dyn_R_armFluff_1!(SimulationSpace:AnimPhysSimSpaceType Enum, ComponentPose=@n10, Alpha=1, ExternalForce="0,0,0") -> (Pose) {
    angularDampingOverride=1
    angularSpringConstant=10000
    bAngularSpring=true
    bOverrideAngularDamping=true
    bUseGravityOverride=true
    gravityScale=0.35
    simulationSpace=RootRelative
    alphaBoolBlend {blendOption=Linear}
    boundBone {boneName=dyn_R_armFluff_1}
    componentLinearAccScale=<1 1 1>
    componentLinearVelScale=<1 1 1>
    gravityOverride {z=-980}
    physicsBodyDefinitions [{
      boundBone {boneName=dyn_R_armFluff_1}
      boxExtents=<25 25 25>
      constraintSetup {angularLimitsMax=<0 0 30>, angularLimitsMin=<0 -45 -30>, angularTarget=<1 0 0>}
      localJointOffset=<2 0 0>
     }]
   }
   n12 = Anim Dynamics - Bone: dyn_R_armFluff_2!(SimulationSpace:AnimPhysSimSpaceType Enum, ComponentPose=@n11, Alpha=1, ExternalForce="0,0,0") -> (Pose) {
    angularDampingOverride=1
    angularSpringConstant=12000
    bAngularSpring=true
    bOverrideAngularDamping=true
    bUseGravityOverride=true
    gravityScale=0.35
    simulationSpace=RootRelative
    alphaBoolBlend {blendOption=Linear}
    boundBone {boneName=dyn_R_armFluff_2}
    componentLinearAccScale=<1 1 1>
    componentLinearVelScale=<1 1 1>
    gravityOverride {z=-1200}
    physicsBodyDefinitions [{
      boundBone {boneName=dyn_R_armFluff_2}
      boxExtents=<20 20 20>
      constraintSetup {angularLimitsMax=<0 0 30>, angularLimitsMin=<0 -45 -30>, angularTarget=<1 0 0>}
      localJointOffset=<2 0 0>
     }]
   }
   n13 = Anim Dynamics - Bone: dyn_R_armFluff_3!(SimulationSpace:AnimPhysSimSpaceType Enum, ComponentPose=@n12, Alpha=1, ExternalForce="0,0,0") -> (Pose) {
    angularDampingOverride=1
    angularSpringConstant=10000
    bAngularSpring=true
    bOverrideAngularDamping=true
    bUseGravityOverride=true
    gravityScale=0.35
    simulationSpace=RootRelative
    alphaBoolBlend {blendOption=Linear}
    boundBone {boneName=dyn_R_armFluff_3}
    componentLinearAccScale=<1 1 1>
    componentLinearVelScale=<1 1 1>
    gravityOverride {z=-980}
    physicsBodyDefinitions [{
      boundBone {boneName=dyn_R_armFluff_3}
      boxExtents=<25 25 25>
      constraintSetup {angularLimitsMax=<0 0 30>, angularLimitsMin=<0 -45 -30>, angularTarget=<1 0 0>}
      localJointOffset=<2 0 0>
     }]
   }
   n14 = Anim Dynamics - Bone: dyn_R_armFluff_4!(SimulationSpace:AnimPhysSimSpaceType Enum, ComponentPose=@n13, Alpha=1, ExternalForce="0,0,0") -> (Pose) {
    angularDampingOverride=1
    angularSpringConstant=12000
    bAngularSpring=true
    bOverrideAngularDamping=true
    bUseGravityOverride=true
    gravityScale=0.35
    simulationSpace=RootRelative
    alphaBoolBlend {blendOption=Linear}
    boundBone {boneName=dyn_R_armFluff_4}
    componentLinearAccScale=<1 1 1>
    componentLinearVelScale=<1 1 1>
    gravityOverride {z=-980}
    physicsBodyDefinitions [{
      boundBone {boneName=dyn_R_armFluff_4}
      boxExtents=<20 20 20>
      constraintSetup {angularLimitsMax=<0 0 30>, angularLimitsMin=<0 -45 -30>, angularTarget=<1 0 0>}
      localJointOffset=<2 0 0>
     }]
   }
   n15 = Anim Dynamics - Bone: dyn_R_armFluff_5!(SimulationSpace:AnimPhysSimSpaceType Enum, ComponentPose=@n14, Alpha=1, ExternalForce="0,0,0") -> (Pose) {
    angularDampingOverride=1
    angularSpringConstant=10000
    bAngularSpring=true
    bOverrideAngularDamping=true
    bUseGravityOverride=true
    gravityScale=0.35
    simulationSpace=RootRelative
    alphaBoolBlend {blendOption=Linear}
    boundBone {boneName=dyn_R_armFluff_5}
    componentLinearAccScale=<1 1 1>
    componentLinearVelScale=<1 1 1>
    gravityOverride {z=-980}
    physicsBodyDefinitions [{
      boundBone {boneName=dyn_R_armFluff_5}
      boxExtents=<25 25 25>
      constraintSetup {angularLimitsMax=<0 0 30>, angularLimitsMin=<0 -45 -30>, angularTarget=<1 0 0>}
      localJointOffset=<2 0 0>
     }]
   }
   n16 = Anim Dynamics - Bone: dyn_R_armFluff_6!(SimulationSpace:AnimPhysSimSpaceType Enum, ComponentPose=@n15, Alpha=1, ExternalForce="0,0,0") -> (Pose) {
    angularDampingOverride=1
    angularSpringConstant=12000
    bAngularSpring=true
    bOverrideAngularDamping=true
    bUseGravityOverride=true
    gravityScale=0.35
    simulationSpace=RootRelative
    alphaBoolBlend {blendOption=Linear}
    boundBone {boneName=dyn_R_armFluff_6}
    componentLinearAccScale=<1 1 1>
    componentLinearVelScale=<1 1 1>
    gravityOverride {z=-980}
    physicsBodyDefinitions [{
      boundBone {boneName=dyn_R_armFluff_6}
      boxExtents=<18 18 18>
      constraintSetup {angularLimitsMax=<0 0 30>, angularLimitsMin=<0 -45 -30>, angularTarget=<1 0 0>}
      localJointOffset=<2 0 0>
     }]
   }
   n17 = Anim Dynamics - Bone: dyn_R_armFluff_7!(SimulationSpace:AnimPhysSimSpaceType Enum, ComponentPose=@n16, Alpha=1, ExternalForce="0,0,0") -> (Pose) {
    angularDampingOverride=1
    angularSpringConstant=10000
    bAngularSpring=true
    bOverrideAngularDamping=true
    bUseGravityOverride=true
    gravityScale=0.35
    simulationSpace=RootRelative
    alphaBoolBlend {blendOption=Linear}
    boundBone {boneName=dyn_R_armFluff_7}
    componentLinearAccScale=<1 1 1>
    componentLinearVelScale=<1 1 1>
    gravityOverride {z=-1200}
    physicsBodyDefinitions [{
      boundBone {boneName=dyn_R_armFluff_7}
      boxExtents=<25 25 25>
      constraintSetup {angularLimitsMax=<0 0 30>, angularLimitsMin=<0 -45 -30>, angularTarget=<1 0 0>}
      localJointOffset=<2 0 0>
     }]
   }
   n18 = Anim Dynamics - Bone: dyn_R_armFluff_8!(SimulationSpace:AnimPhysSimSpaceType Enum, ComponentPose=@n17, Alpha=1, ExternalForce="0,0,0") -> (Pose) {
    angularDampingOverride=1
    angularSpringConstant=12000
    bAngularSpring=true
    bOverrideAngularDamping=true
    bUseGravityOverride=true
    gravityScale=0.35
    simulationSpace=RootRelative
    alphaBoolBlend {blendOption=Linear}
    boundBone {boneName=dyn_R_armFluff_8}
    componentLinearAccScale=<1 1 1>
    componentLinearVelScale=<1 1 1>
    gravityOverride {z=-980}
    physicsBodyDefinitions [{
      boundBone {boneName=dyn_R_armFluff_8}
      boxExtents=<18 18 18>
      constraintSetup {angularLimitsMax=<0 0 30>, angularLimitsMin=<0 -45 -30>, angularTarget=<1 0 0>}
      localJointOffset=<2 0 0>
     }]
   }
   n19 = Anim Dynamics - Bone: dyn_L_legCurtain_outer_1!(SimulationSpace:AnimPhysSimSpaceType Enum, ComponentPose=@n18, Alpha=1, ExternalForce="0,0,0") -> (Pose) {
    angularDampingOverride=1
    angularSpringConstant=9000
    bAngularSpring=true
    bOverrideAngularDamping=true
    bUseGravityOverride=true
    gravityScale=0.35
    simulationSpace=RootRelative
    alphaBoolBlend {blendOption=Linear}
    boundBone {boneName=dyn_L_legCurtain_outer_1}
    componentLinearAccScale=<1 1 1>
    componentLinearVelScale=<1 1 1>
    gravityOverride {z=-980}
    physicsBodyDefinitions [{
      boundBone {boneName=dyn_L_legCurtain_outer_1}
      boxExtents=<20 20 20>
      constraintSetup {angularLimitsMax=<3 3 0>, angularLimitsMin=<-3 -3 -10>, angularTarget=<1 0 0>}
      localJointOffset=<2 0 0>
     }]
   }
   n20 = Anim Dynamics - Bone: dyn_L_legCurtain_outer_2!(SimulationSpace:AnimPhysSimSpaceType Enum, ComponentPose=@n19, Alpha=1, ExternalForce="0,0,0") -> (Pose) {
    angularDampingOverride=1
    angularSpringConstant=9000
    bAngularSpring=true
    bOverrideAngularDamping=true
    bUseGravityOverride=true
    gravityScale=0.35
    simulationSpace=RootRelative
    alphaBoolBlend {blendOption=Linear}
    boundBone {boneName=dyn_L_legCurtain_outer_2}
    componentLinearAccScale=<1 1 1>
    componentLinearVelScale=<1 1 1>
    gravityOverride {z=-980}
    physicsBodyDefinitions [{
      boundBone {boneName=dyn_L_legCurtain_outer_2}
      boxExtents=<20 20 20>
      constraintSetup {angularLimitsMax=<3 3 0>, angularLimitsMin=<-3 -3 -10>, angularTarget=<1 0 0>}
      localJointOffset=<2 0 0>
     }]
   }
   n21 = Anim Dynamics - Bone: dyn_L_legCurtain_inner_1!(SimulationSpace:AnimPhysSimSpaceType Enum, ComponentPose=@n20, Alpha=1, ExternalForce="0,0,0") -> (Pose) {
    angularDampingOverride=1
    angularSpringConstant=9000
    bAngularSpring=true
    bOverrideAngularDamping=true
    bUseGravityOverride=true
    gravityScale=0.35
    simulationSpace=RootRelative
    alphaBoolBlend {blendOption=Linear}
    boundBone {boneName=dyn_L_legCurtain_inner_1}
    componentLinearAccScale=<1 1 1>
    componentLinearVelScale=<1 1 1>
    gravityOverride {z=-980}
    physicsBodyDefinitions [{
      boundBone {boneName=dyn_L_legCurtain_inner_1}
      boxExtents=<20 20 20>
      constraintSetup {angularLimitsMax=<3 0 0>, angularLimitsMin=<-3 0 -10>, angularTarget=<1 0 0>}
      localJointOffset=<2 0 0>
     }]
   }
   n22 = Anim Dynamics - Bone: dyn_L_legCurtain_inner_2!(SimulationSpace:AnimPhysSimSpaceType Enum, ComponentPose=@n21, Alpha=1, ExternalForce="0,0,0") -> (Pose) {
    angularDampingOverride=1
    angularSpringConstant=9000
    bAngularSpring=true
    bOverrideAngularDamping=true
    bUseGravityOverride=true
    gravityScale=0.35
    simulationSpace=RootRelative
    alphaBoolBlend {blendOption=Linear}
    boundBone {boneName=dyn_L_legCurtain_inner_2}
    componentLinearAccScale=<1 1 1>
    componentLinearVelScale=<1 1 1>
    gravityOverride {z=-980}
    physicsBodyDefinitions [{
      boundBone {boneName=dyn_L_legCurtain_inner_2}
      boxExtents=<20 20 20>
      constraintSetup {angularLimitsMax=<3 0 0>, angularLimitsMin=<-3 0 -10>, angularTarget=<1 0 0>}
      localJointOffset=<2 0 0>
     }]
   }
   n23 = Anim Dynamics - Bone: dyn_L_legCurtain_front_1!(SimulationSpace:AnimPhysSimSpaceType Enum, ComponentPose=@n22, Alpha=1, ExternalForce="0,0,0") -> (Pose) {
    angularDampingOverride=1
    angularSpringConstant=9000
    bAngularSpring=true
    bOverrideAngularDamping=true
    bUseGravityOverride=true
    gravityScale=0.35
    simulationSpace=RootRelative
    alphaBoolBlend {blendOption=Linear}
    boundBone {boneName=dyn_L_legCurtain_front_1}
    componentLinearAccScale=<1 1 1>
    componentLinearVelScale=<1 1 1>
    gravityOverride {z=-980}
    physicsBodyDefinitions [{
      boundBone {boneName=dyn_L_legCurtain_front_1}
      boxExtents=<20 20 20>
      constraintSetup {angularLimitsMax=<3 3 0>, angularLimitsMin=<-3 -3 -15>, angularTarget=<1 0 0>}
      localJointOffset=<2 0 0>
     }]
   }
   n24 = Anim Dynamics - Bone: dyn_L_legCurtain_front_2!(SimulationSpace:AnimPhysSimSpaceType Enum, ComponentPose=@n23, Alpha=1, ExternalForce="0,0,0") -> (Pose) {
    angularDampingOverride=1
    angularSpringConstant=9000
    bAngularSpring=true
    bOverrideAngularDamping=true
    bUseGravityOverride=true
    gravityScale=0.35
    simulationSpace=RootRelative
    alphaBoolBlend {blendOption=Linear}
    boundBone {boneName=dyn_L_legCurtain_front_2}
    componentLinearAccScale=<1 1 1>
    componentLinearVelScale=<1 1 1>
    gravityOverride {z=-980}
    physicsBodyDefinitions [{
      boundBone {boneName=dyn_L_legCurtain_front_2}
      boxExtents=<20 20 20>
      constraintSetup {angularLimitsMax=<3 3 0>, angularLimitsMin=<-3 -3 -15>, angularTarget=<1 0 0>}
      localJointOffset=<2 0 0>
     }]
   }
   n25 = Anim Dynamics - Bone: dyn_L_legCurtain_back_1!(SimulationSpace:AnimPhysSimSpaceType Enum, ComponentPose=@n24, Alpha=1, ExternalForce="0,0,0") -> (Pose) {
    angularDampingOverride=1
    angularSpringConstant=9000
    bAngularSpring=true
    bOverrideAngularDamping=true
    bUseGravityOverride=true
    gravityScale=0.35
    simulationSpace=RootRelative
    alphaBoolBlend {blendOption=Linear}
    boundBone {boneName=dyn_L_legCurtain_back_1}
    componentLinearAccScale=<1 1 1>
    componentLinearVelScale=<1 1 1>
    gravityOverride {z=-980}
    physicsBodyDefinitions [{
      boundBone {boneName=dyn_L_legCurtain_back_1}
      boxExtents=<20 20 20>
      constraintSetup {angularLimitsMax=<3 3 0>, angularLimitsMin=<-3 -3 -15>, angularTarget=<1 0 0>}
      localJointOffset=<2 0 0>
     }]
   }
   n26 = Anim Dynamics - Bone: dyn_L_legCurtain_back_2!(SimulationSpace:AnimPhysSimSpaceType Enum, ComponentPose=@n25, Alpha=1, ExternalForce="0,0,0") -> (Pose) {
    angularDampingOverride=1
    angularSpringConstant=9000
    bAngularSpring=true
    bOverrideAngularDamping=true
    bUseGravityOverride=true
    gravityScale=0.35
    simulationSpace=RootRelative
    alphaBoolBlend {blendOption=Linear}
    boundBone {boneName=dyn_L_legCurtain_back_2}
    componentLinearAccScale=<1 1 1>
    componentLinearVelScale=<1 1 1>
    gravityOverride {z=-980}
    physicsBodyDefinitions [{
      boundBone {boneName=dyn_L_legCurtain_back_2}
      boxExtents=<20 20 20>
      constraintSetup {angularLimitsMax=<3 3 3>, angularLimitsMin=<-3 -3 -15>, angularTarget=<1 0 0>}
      localJointOffset=<2 0 0>
     }]
   }
   n27 = Anim Dynamics - Bone: dyn_R_legCurtain_outer_1!(SimulationSpace:AnimPhysSimSpaceType Enum, ComponentPose=@n26, Alpha=1, ExternalForce="0,0,0") -> (Pose) {
    angularDampingOverride=1
    angularSpringConstant=9000
    bAngularSpring=true
    bOverrideAngularDamping=true
    bUseGravityOverride=true
    gravityScale=0.35
    simulationSpace=RootRelative
    alphaBoolBlend {blendOption=Linear}
    boundBone {boneName=dyn_R_legCurtain_outer_1}
    componentLinearAccScale=<1 1 1>
    componentLinearVelScale=<1 1 1>
    gravityOverride {z=-980}
    physicsBodyDefinitions [{
      boundBone {boneName=dyn_R_legCurtain_outer_1}
      boxExtents=<20 20 20>
      constraintSetup {angularLimitsMax=<3 3 0>, angularLimitsMin=<-3 -3 -10>, angularTarget=<1 0 0>}
      localJointOffset=<-2 0 0>
     }]
   }
   n28 = Anim Dynamics - Bone: dyn_R_legCurtain_outer_2!(SimulationSpace:AnimPhysSimSpaceType Enum, ComponentPose=@n27, Alpha=1, ExternalForce="0,0,0") -> (Pose) {
    angularDampingOverride=1
    angularSpringConstant=9000
    bAngularSpring=true
    bOverrideAngularDamping=true
    bUseGravityOverride=true
    gravityScale=0.35
    simulationSpace=RootRelative
    alphaBoolBlend {blendOption=Linear}
    boundBone {boneName=dyn_R_legCurtain_outer_2}
    componentLinearAccScale=<1 1 1>
    componentLinearVelScale=<1 1 1>
    gravityOverride {z=-980}
    physicsBodyDefinitions [{
      boundBone {boneName=dyn_R_legCurtain_outer_2}
      boxExtents=<20 20 20>
      constraintSetup {angularLimitsMax=<3 3 0>, angularLimitsMin=<-3 -3 -10>, angularTarget=<1 0 0>}
      localJointOffset=<-2 0 0>
     }]
   }
   n29 = Anim Dynamics - Bone: dyn_R_legCurtain_inner_1!(SimulationSpace:AnimPhysSimSpaceType Enum, ComponentPose=@n28, Alpha=1, ExternalForce="0,0,0") -> (Pose) {
    angularDampingOverride=1
    angularSpringConstant=9000
    bAngularSpring=true
    bOverrideAngularDamping=true
    bUseGravityOverride=true
    gravityScale=0.35
    simulationSpace=RootRelative
    alphaBoolBlend {blendOption=Linear}
    boundBone {boneName=dyn_R_legCurtain_inner_1}
    componentLinearAccScale=<1 1 1>
    componentLinearVelScale=<1 1 1>
    gravityOverride {z=-980}
    physicsBodyDefinitions [{
      boundBone {boneName=dyn_R_legCurtain_inner_1}
      boxExtents=<20 20 20>
      constraintSetup {angularLimitsMax=<3 0 0>, angularLimitsMin=<-3 0 -10>, angularTarget=<1 0 0>}
      localJointOffset=<-2 0 0>
     }]
   }
   n30 = Anim Dynamics - Bone: dyn_R_legCurtain_inner_2!(SimulationSpace:AnimPhysSimSpaceType Enum, ComponentPose=@n29, Alpha=1, ExternalForce="0,0,0") -> (Pose) {
    angularDampingOverride=1
    angularSpringConstant=9000
    bAngularSpring=true
    bOverrideAngularDamping=true
    bUseGravityOverride=true
    gravityScale=0.35
    simulationSpace=RootRelative
    alphaBoolBlend {blendOption=Linear}
    boundBone {boneName=dyn_R_legCurtain_inner_2}
    componentLinearAccScale=<1 1 1>
    componentLinearVelScale=<1 1 1>
    gravityOverride {z=-980}
    physicsBodyDefinitions [{
      boundBone {boneName=dyn_R_legCurtain_inner_2}
      boxExtents=<20 20 20>
      constraintSetup {angularLimitsMax=<3 0 0>, angularLimitsMin=<-3 0 -10>, angularTarget=<1 0 0>}
      localJointOffset=<-2 0 0>
     }]
   }
   n31 = Anim Dynamics - Bone: dyn_R_legCurtain_front_1!(SimulationSpace:AnimPhysSimSpaceType Enum, ComponentPose=@n30, Alpha=1, ExternalForce="0,0,0") -> (Pose) {
    angularDampingOverride=1
    angularSpringConstant=9000
    bAngularSpring=true
    bOverrideAngularDamping=true
    bUseGravityOverride=true
    gravityScale=0.35
    simulationSpace=RootRelative
    alphaBoolBlend {blendOption=Linear}
    boundBone {boneName=dyn_R_legCurtain_front_1}
    componentLinearAccScale=<1 1 1>
    componentLinearVelScale=<1 1 1>
    gravityOverride {z=-980}
    physicsBodyDefinitions [{
      boundBone {boneName=dyn_R_legCurtain_front_1}
      boxExtents=<20 20 20>
      constraintSetup {angularLimitsMax=<3 3 0>, angularLimitsMin=<-3 -3 -15>, angularTarget=<1 0 0>}
      localJointOffset=<-2 0 0>
     }]
   }
   n32 = Anim Dynamics - Bone: dyn_R_legCurtain_front_2!(SimulationSpace:AnimPhysSimSpaceType Enum, ComponentPose=@n31, Alpha=1, ExternalForce="0,0,0") -> (Pose) {
    angularDampingOverride=1
    angularSpringConstant=9000
    bAngularSpring=true
    bOverrideAngularDamping=true
    bUseGravityOverride=true
    gravityScale=0.35
    simulationSpace=RootRelative
    alphaBoolBlend {blendOption=Linear}
    boundBone {boneName=dyn_R_legCurtain_front_2}
    componentLinearAccScale=<1 1 1>
    componentLinearVelScale=<1 1 1>
    gravityOverride {z=-980}
    physicsBodyDefinitions [{
      boundBone {boneName=dyn_R_legCurtain_front_2}
      boxExtents=<20 20 20>
      constraintSetup {angularLimitsMax=<3 3 0>, angularLimitsMin=<-3 -3 -15>, angularTarget=<1 0 0>}
      localJointOffset=<-2 0 0>
     }]
   }
   n33 = Anim Dynamics - Bone: dyn_R_legCurtain_back_1!(SimulationSpace:AnimPhysSimSpaceType Enum, ComponentPose=@n32, Alpha=1, ExternalForce="0,0,0") -> (Pose) {
    angularDampingOverride=1
    angularSpringConstant=9000
    bAngularSpring=true
    bOverrideAngularDamping=true
    bUseGravityOverride=true
    gravityScale=0.35
    simulationSpace=RootRelative
    alphaBoolBlend {blendOption=Linear}
    boundBone {boneName=dyn_R_legCurtain_back_1}
    componentLinearAccScale=<1 1 1>
    componentLinearVelScale=<1 1 1>
    gravityOverride {z=-980}
    physicsBodyDefinitions [{
      boundBone {boneName=dyn_R_legCurtain_back_1}
      boxExtents=<20 20 20>
      constraintSetup {angularLimitsMax=<3 3 0>, angularLimitsMin=<-3 -3 -15>, angularTarget=<1 0 0>}
      localJointOffset=<-2 0 0>
     }]
   }
   n34 = Anim Dynamics - Bone: dyn_R_legCurtain_back_2!(SimulationSpace:AnimPhysSimSpaceType Enum, ComponentPose=@n33, Alpha=1, ExternalForce="0,0,0") -> (Pose) {
    angularDampingOverride=1
    angularSpringConstant=9000
    bAngularSpring=true
    bOverrideAngularDamping=true
    bUseGravityOverride=true
    gravityScale=0.35
    simulationSpace=RootRelative
    alphaBoolBlend {blendOption=Linear}
    boundBone {boneName=dyn_R_legCurtain_back_2}
    componentLinearAccScale=<1 1 1>
    componentLinearVelScale=<1 1 1>
    gravityOverride {z=-980}
    physicsBodyDefinitions [{
      boundBone {boneName=dyn_R_legCurtain_back_2}
      boxExtents=<20 20 20>
      constraintSetup {angularLimitsMax=<3 3 3>, angularLimitsMin=<-3 -3 -15>, angularTarget=<1 0 0>}
      localJointOffset=<-2 0 0>
     }]
   }
   n35 = "Constraint - dyn_holster (PP)"!(ComponentPose=@n34, Alpha=1, ConstraintWeights_0=0.1, ConstraintWeights_1:float=0.9) -> (Pose) {
    alphaBoolBlend {blendOption=Linear}
    boneToModify {boneName=dyn_holster}
    constraintSetup [{targetBone {boneName=pelvis}}, {targetBone {boneName=thigh_l}}]
    constraintWeights [1,1]
   }
   n36 = Anim Dynamics - Bone: dyn_back_collarFluff(ComponentPose=@n35, Alpha=1, ExternalForce="0,0,0") {
    angularDampingOverride=1
    angularSpringConstant=10000
    bAngularSpring=true
    bOverrideAngularDamping=true
    bUseGravityOverride=true
    gravityScale=0.35
    linearDampingOverride=1
    linearSpringConstant=500
    simulationSpace=RootRelative
    alphaBoolBlend {blendOption=Linear}
    boundBone {boneName=dyn_back_collarFluff}
    componentAppliedLinearAccClamp=<1000 1000 1000>
    componentLinearAccScale=<0.1 0.1 0.1>
    componentLinearVelScale=<0.1 0.1 0.1>
    gravityOverride {z=-980}
    physicsBodyDefinitions [{
      boundBone {boneName=dyn_back_collarFluff}
      boxExtents=<25 25 25>
      constraintSetup {angularLimitsMax=<3 30 3>, angularLimitsMin=<-3 -20 -3>, angularTarget=<1 0 0>}
      localJointOffset=<1 0 0>
     }]
   }
   n37 = Anim Dynamics - Bone: dyn_L_backJacket!(SimulationSpace:AnimPhysSimSpaceType Enum, ComponentPose=@n36, Alpha=1, ExternalForce="0,0,0") -> (Pose) {
    angularDampingOverride=0.9999
    angularSpringConstant=15000
    bAngularSpring=true
    bOverrideAngularDamping=true
    bUseGravityOverride=true
    gravityScale=0.35
    simulationSpace=RootRelative
    alphaBoolBlend {blendOption=Linear}
    boundBone {boneName=dyn_L_backJacket}
    componentLinearAccScale=<1 1 1>
    componentLinearVelScale=<1 1 1>
    gravityOverride {z=-980}
    physicsBodyDefinitions [{
      boundBone {boneName=dyn_L_backJacket}
      boxExtents=<20 20 20>
      constraintSetup {angularLimitsMax=<5 10 10>, angularLimitsMin=<-5 -10 -60>, angularTarget=<1 0 0>}
      localJointOffset=<2 0 0>
     }]
   }
   n38 = Anim Dynamics - Bone: dyn_R_backJacket!(SimulationSpace:AnimPhysSimSpaceType Enum, ComponentPose=@n37, Alpha=1, ExternalForce="0,0,0") -> (Pose) {
    angularDampingOverride=0.9999
    angularSpringConstant=15000
    bAngularSpring=true
    bOverrideAngularDamping=true
    bUseGravityOverride=true
    gravityScale=0.35
    simulationSpace=RootRelative
    alphaBoolBlend {blendOption=Linear}
    boundBone {boneName=dyn_R_backJacket}
    componentLinearAccScale=<1 1 1>
    componentLinearVelScale=<1 1 1>
    gravityOverride {z=-980}
    physicsBodyDefinitions [{
      boundBone {boneName=dyn_R_backJacket}
      boxExtents=<20 20 20>
      constraintSetup {angularLimitsMax=<5 10 10>, angularLimitsMin=<-5 -10 -60>, angularTarget=<1 0 0>}
      localJointOffset=<2 0 0>
     }]
   }
   n39 = Anim Dynamics - Bone: dyn_L_shoulderFur_1(ComponentPose=@n38, Alpha=1, ExternalForce="0,0,0") {
    angularDampingOverride=1
    angularSpringConstant=5000
    bAngularSpring=true
    bOverrideAngularDamping=true
    bUseGravityOverride=true
    gravityScale=0.35
    linearDampingOverride=1
    simulationSpace=RootRelative
    alphaBoolBlend {blendOption=Linear}
    boundBone {boneName=dyn_L_shoulderFur_1}
    componentAppliedLinearAccClamp=<1000 1000 1000>
    componentLinearAccScale=<0.1 0.1 0.1>
    componentLinearVelScale=<0.1 0.1 0.1>
    gravityOverride {z=-980}
    physicsBodyDefinitions [{
      boundBone {boneName=dyn_L_shoulderFur_1}
      boxExtents=<22 22 22>
      constraintSetup {angularLimitsMax=<0 5 10>, angularLimitsMin=<0 -30 -10>, angularTarget=<1 0 0>}
      localJointOffset=<1 0 0>
     }]
   }
   n40 = Anim Dynamics - Bone: dyn_L_shoulderFur_2(ComponentPose=@n39, Alpha=1, ExternalForce="0,0,0") {
    angularDampingOverride=1
    angularSpringConstant=5000
    bAngularSpring=true
    bOverrideAngularDamping=true
    bUseGravityOverride=true
    gravityScale=0.35
    linearDampingOverride=1
    simulationSpace=RootRelative
    alphaBoolBlend {blendOption=Linear}
    boundBone {boneName=dyn_L_shoulderFur_2}
    componentAppliedLinearAccClamp=<1000 1000 1000>
    componentLinearAccScale=<0.1 0.1 0.1>
    componentLinearVelScale=<0.1 0.1 0.1>
    gravityOverride {z=-980}
    physicsBodyDefinitions [{
      boundBone {boneName=dyn_L_shoulderFur_2}
      boxExtents=<22 22 22>
      constraintSetup {angularLimitsMax=<0 5 10>, angularLimitsMin=<0 -30 -10>, angularTarget=<1 0 0>}
      localJointOffset=<1 0 0>
     }]
   }
   n41 = Anim Dynamics - Bone: dyn_R_shoulderFur_1(ComponentPose=@n40, Alpha=1, ExternalForce="0,0,0") {
    angularDampingOverride=1
    angularSpringConstant=5000
    bAngularSpring=true
    bOverrideAngularDamping=true
    bUseGravityOverride=true
    gravityScale=0.35
    linearDampingOverride=1
    simulationSpace=RootRelative
    alphaBoolBlend {blendOption=Linear}
    boundBone {boneName=dyn_R_shoulderFur_1}
    componentAppliedLinearAccClamp=<1000 1000 1000>
    componentLinearAccScale=<0.1 0.1 0.1>
    componentLinearVelScale=<0.1 0.1 0.1>
    gravityOverride {z=-980}
    physicsBodyDefinitions [{
      boundBone {boneName=dyn_R_shoulderFur_1}
      boxExtents=<22 22 22>
      constraintSetup {angularLimitsMax=<0 5 10>, angularLimitsMin=<0 -30 -10>, angularTarget=<1 0 0>}
      localJointOffset=<-1 0 0>
     }]
   }
   n42 = Anim Dynamics - Bone: dyn_R_shoulderFur_2(ComponentPose=@n41, Alpha=1, ExternalForce="0,0,0") {
    angularDampingOverride=1
    angularSpringConstant=5000
    bAngularSpring=true
    bOverrideAngularDamping=true
    bUseGravityOverride=true
    gravityScale=0.35
    linearDampingOverride=1
    simulationSpace=RootRelative
    alphaBoolBlend {blendOption=Linear}
    boundBone {boneName=dyn_R_shoulderFur_2}
    componentAppliedLinearAccClamp=<1000 1000 1000>
    componentLinearAccScale=<0.1 0.1 0.1>
    componentLinearVelScale=<0.1 0.1 0.1>
    gravityOverride {z=-980}
    physicsBodyDefinitions [{
      boundBone {boneName=dyn_R_shoulderFur_2}
      boxExtents=<22 22 22>
      constraintSetup {angularLimitsMax=<0 5 10>, angularLimitsMin=<0 -30 -10>, angularTarget=<1 0 0>}
      localJointOffset=<-1 0 0>
     }]
   }
   n43 = "Constraint - dyn_L_legCurtain_front_footFold (P)"(ComponentPose=@n42, Alpha=1, ConstraintWeights_0=1) {
    alphaBoolBlend {blendOption=Linear}
    boneToModify {boneName=dyn_L_legCurtain_front_footFold}
    constraintSetup [{targetBone {boneName=foot_l}}]
    constraintWeights [1]
   }
   n44 = "Constraint - dyn_R_legCurtain_front_footFold (P)"(ComponentPose=@n43, Alpha=1, ConstraintWeights_0=1) {
    alphaBoolBlend {blendOption=Linear}
    boneToModify {boneName=dyn_R_legCurtain_front_footFold}
    constraintSetup [{targetBone {boneName=foot_r}}]
    constraintWeights [1]
   }
   n45 = "Constraint - dyn_collarConstraint (PP)"!(ComponentPose=@n44, Alpha=1, ConstraintWeights_0=0.2, ConstraintWeights_1:float=0.8) -> (Pose) {
    alphaBoolBlend {blendOption=Linear}
    alphaScaleBiasClamp {scale=0.7}
    boneToModify {boneName=dyn_collarConstraint}
    constraintSetup [{targetBone {boneName=neck_01}}, {targetBone {boneName=neck_02}}]
    constraintWeights [1,1]
   }
   n46 = Control Rig(Source=Component To Local(ComponentPose=@n45)) {
    alphaBoolBlend {blendOption=Linear}
    controlRigAssetReference {blueprintRigClass=&/F_MED_MadameMoth_Ember_CtrlRig.F_MED_MadameMoth_Ember_CtrlRig_C}
   }
   Save cached pose 'startpose'(Pose=@n46) {cacheName=startpose}
  ]}
 EventGraph {nodes [
   n1 = Event BlueprintUpdateAnimation()
   > n2 = Sequence!(execute) -> (then_0, then_1, then_2:Exec, then_3:Exec, then_4:Exec)
   Set simSpace(execute=@n2.then_2, simSpace:AnimPhysSimSpaceType Enum) -> (Output_Get:AnimPhysSimSpaceType Enum)
   > Branch(Condition=OR Boolean(A=Get bIsDBNO(), B=Get bIsInFrontEnd()))
   > n3 = Branch(Condition=OR Boolean(A=Get bIsDBNOCarried(), B=Get bIsInFrontEnd()))
   > Set simSpace(simSpace=World)
   Set simSpace(execute=@n3.else)
   Branch(execute=@n2.then_4, Condition=Get bIsInFrontEnd())
   > Set Custom Idle is Playing(Custom Idle is Playing:float=GetCurveValue(self=GetOwnerFrontEndAnimInstance(), CurveName=Frontend_Idle_Bypass)) -> (Output_Get:float)
   Is Valid(exec=@n2.then_0, InputObject=Get SkelMesh())
   > Set WPO Globals Function(
    Material=Get Mat Instance Dynamic()
    Overlay Material=Get Overlay Mat Instance Dynamic()
    Skeletal Mesh=Get SkelMesh()
    "Local Noise Transform  Location"=Get Local Noise Transform Location() -> (Local Noise Transform Location:Vector)
    Local Noise Offset=Get Local Noise Offset()
   )
   > Set WPO Regional Function!(
    execute
    Material:Material Instance Dynamic Object Reference=Get Mat Instance Dynamic()
    Overlay Material:Material Instance Dynamic Object Reference=Get Overlay Mat Instance Dynamic()
    Skeletal Mesh:Skeletal Mesh Component Object Reference=Get SkelMesh()
    Velocity Transform:Name=foot_r
    Velocity Interp Speed:double=Get Velocity Interp Speed()
    Collision Transform:Name=calf_r
    Collision Transform Axis:EAxis Enum
    Invert Collision Transform Axis:bool=false
    Delta Time X:double=@n1.DeltaTimeX
    Previous Velocity Transform Location:"Vector (by ref)"=Get R1 Velocity Transform Location()
    Previous Velocity Transform Velocity:"Vector (by ref)"=Get R1 Velocity Transform Velocity()
    Region:int=1
   ) -> (then)
   > Set WPO Regional Function(
    Material=Get Mat Instance Dynamic()
    Overlay Material=Get Overlay Mat Instance Dynamic()
    Skeletal Mesh=Get SkelMesh()
    Velocity Transform=foot_l
    Velocity Interp Speed=Get Velocity Interp Speed()
    Collision Transform=calf_l
    Invert Collision Transform Axis=true
    Delta Time X=@n1.DeltaTimeX
    Previous Velocity Transform Location=Get R2 Velocity Transform Location()
    Previous Velocity Transform Velocity=Get R2 Velocity Transform Velocity()
    Region=2
   )
   > n4 = Sequence!(execute) -> (then_0, then_1)
   > n5 = Branch(Condition=Get bIsSkyDiving())
   > n6 = Set WPO Pho Wind Intensity(WPO Pho Wind Intensity=0.7)
   SetScalarParameterValue(
    execute=[@n7.then, @n6.then]
    self=[Get Mat Instance Dynamic(), Get Overlay Mat Instance Dynamic()]
    ParameterName=ClothSim_WindStrength
    Value=Get WPO Pho Wind Intensity()
   )
   n7 = Set WPO Pho Wind Intensity(execute=@n5.else, WPO Pho Wind Intensity=0)
   n8 = Branch(execute=@n4.then_1, Condition=Get bIsSkyDiving())
   > n9 = Set ClothSim_VelocityMax(ClothSim_VelocityMax:double=750) -> (Output_Get:double)
   SetScalarParameterValue(
    execute=[@n10.then, @n9.then]
    self=Get Mat Instance Dynamic()
    ParameterName=ClothSim_VelocityMax
    Value=Get ClothSim_VelocityMax() -> (ClothSim_VelocityMax:double)
   )
   n10 = Set ClothSim_VelocityMax(execute=@n8.else, ClothSim_VelocityMax=5000)
   Set frontendAlpha(execute=@n2.then_3, frontendAlpha:double=SelectFloat(A=0.7, B=1, bPickA=Get bIsInFrontEnd())) -> (Output_Get:double)
   n11 = Set L foot constraint(execute=@n2.then_1, L foot constraint:double=1) -> (Output_Get:double)
   n12 = GetOwnerSkeletalMeshComponent()
   n13 = BreakRotator(InRot=BreakTransform(InTransform=GetDeltaTransformFromRefPose(self=@n12, BoneName=calf_l, BaseName=dyn_L_legCurtain_front_2)).Rotation)
   > Set L foot constraint(L foot constraint=float - float(A=@n11, B=MapRangeClamped(Value=@n13.Yaw, InRangeA=0, InRangeB=-20, OutRangeA=1, OutRangeB=0)))
   > n14 = Set R foot constraint(R foot constraint:double=1) -> (Output_Get:double)
   n15 = BreakRotator(InRot=BreakTransform(InTransform=GetDeltaTransformFromRefPose(self=@n12, BoneName=calf_r, BaseName=dyn_R_legCurtain_front_2)).Rotation)
   > Set R foot constraint(R foot constraint=float - float(A=@n14, B=MapRangeClamped(Value=@n15.Yaw, InRangeA=0, InRangeB=-20, OutRangeA=1, OutRangeB=0)))
   Event OnCharacterCustomizationLoaded()
   > n16 = Set SkelMesh(SkelMesh=GetPartSkeletalMeshComponent())
   > n17 = Get Dynamic Material Instance(SkeletalMeshComponent=@n16)
   > Set Mat Instance Dynamic(Mat Instance Dynamic=@n17)
   > n18 = GetOrCreateDynamicOverlayMaterial(MeshComponent=@n16, MaterialName=None)
   > Set Overlay Mat Instance Dynamic(Overlay Mat Instance Dynamic=@n18)
  ]}
]
variables [
 ClothSim_VelocityMax {type=number}
 Custom Idle is Playing {type=number}
 frontendAlpha {type=number}
 L foot constraint {type=number}
 Local Noise Offset {type=Vector}
 Local Noise Transform Location {type=Vector}
 Mat Instance Dynamic {type=MaterialInstanceDynamic}
 Overlay Mat Instance Dynamic {type=MaterialInstanceDynamic}
 R foot constraint {type=number}
 R1 Velocity Transform Location {type=Vector}
 R1 Velocity Transform Velocity {type=Vector}
 R2 Velocity Transform Location {type=Vector}
 R2 Velocity Transform Velocity {type=Vector}
 simSpace {type=AnimPhysSimSpaceType}
 SkelMesh {type=SkeletalMeshComponent}
 Velocity Interp Speed {type=number, default=10}
 WPO Pho Wind Intensity {type=number}
]
